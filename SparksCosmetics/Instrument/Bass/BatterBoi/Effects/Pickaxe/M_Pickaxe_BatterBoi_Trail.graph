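# Translucent, unlit, two-sided trail material for Niagara ribbons and mesh particles in the BatterBoi pickaxe effect.
# Particle color, optionally multiplied by a curve-atlas gradient, drives emissive.
# Opacity combines panned, noise-distorted textures with UV gradients and optional masks or dissolve, then applies depth fade or view-angle fade by shading path.
# World position offset sums optional spline thickening, noise waves and bending, with Niagara dynamic parameters controlling wave and bend strength.
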Material /SparksCosmetics/Instrument/Bass/BatterBoi/Effects/Pickaxe/M_Pickaxe_BatterBoi_Trail
graph {
 properties {
  bAutomaticallySetUsageInEditor=false
  bUsedWithNiagaraMeshParticles=true
  bUsedWithNiagaraRibbons=true
  blendMode=BLEND_Translucent
  shadingModel=MSM_Unlit
  twoSided=true
  usedShadingModels=Unlit
 }
 nodes [
  n1 = MaterialFunctionCall(
   WidthBase=ScalarParameter() {defaultValue=0.4, parameterName=SplineBase}
   WidthTip=ScalarParameter() {defaultValue=0.4, parameterName=SplineTip}
   Expand U or V UV Channel=StaticBool()
  ) {materialFunction=&SplineThicken}
  n2 = Constant()
  n3 = ScalarParameter() {defaultValue=1, parameterName=Use_SplineThicken}
  Add(A=Add(A=LinearInterpolate(A=@n1.WorldPositionOffset, B=@n2, Alpha=@n3), B=NamedRerouteUsage()), B=NamedRerouteUsage())
  n4 = ParticleColor()
  n5 = TextureCoordinate()
  n6 = Constant2Vector() {g=0.5, r=0.5}
  n7 = Constant() {r=0.5}
  n8 = Constant() {r=0.5}
  n9 = TextureCoordinate()
  n10 = CurveAtlasRowParameter(CurveTime=LinearInterpolate(
    A=SphereMask(A=@n5, B=@n6, Radius=@n7, Hardness=@n8)
    B=ComponentMask(None=@n9) {r=true}
    Alpha=ScalarParameter() {defaultValue=1, parameterName=Use_ColorCurveMask}
   )) {
   atlas=/BRCosmetics/Weapons/FORT_Melee/Pickaxe_LivelyDomino/FX/CAtlas_Pickaxe_LivelyDomino_ColorCurve
   curve=/BRCosmetics/Weapons/FORT_Melee/Pickaxe_LivelyDomino/FX/CV_Pickaxe_LivelyDomino_ColorCurve_Center
   group=Emissive
   parameterName=ColorCurve
   sortPriority=0
  }
  n11 = Multiply(A=@n10, B=@n4.RGB)
  n12 = ScalarParameter() {parameterName=Use_ColorCurve}
  MaterialFunctionCall(Emissive=Multiply(A=LinearInterpolate(A=@n4.RGB, B=@n11, Alpha=@n12), B=ScalarParameter() {defaultValue=1, parameterName=All light})) {materialFunction=&MF_FX_EmissiveAdjustments}
  n13 = DynamicParameter() {defaultValue=<rgba 1 0 0 6>, paramNames [Random Offset, BaseTextureV_Add, WPO Wave Intensity, SweepTexture_Add]}
  n14 = VectorParameter() {group=WPO, parameterName=WPO Side UV, sortPriority=6, defaultValue=<rgba 1 1 -0.1 1>}
  n15 = TextureCoordinate()
  n16 = Multiply(A=@n15, B=AppendVector(A=@n14.R, B=@n14.R))
  n17 = AppendVector(A=@n14.B, B=@n14.B)
  n18 = TextureSample(UVs=Add(A=AppendVector(A=@n13.Random Offset, B=@n13.Random Offset), B=Panner(Coordinate=@n16, Speed=@n17))) {samplerType=SAMPLERTYPE_Normal, texture=/Game/Characters/Player/Female/Medium/Bodies/F_MED_Bistro_Astronaut/FX/T_GaussianNoise_Normal}
  n19 = DynamicParameter() {defaultValue=<rgba 1 0 0 6>, paramNames [Random Offset, BaseTextureV_Add, WPO Wave Intensity, SweepTexture_Add]}
  n20 = VectorParameter() {group=WPO, parameterName=WPO Wave UV, sortPriority=6, defaultValue=<rgba 1 1 0 2>}
  n21 = AppendVector(A=@n20.R, B=@n20.R)
  n22 = Add(
   A=AppendVector(A=@n19.Random Offset, B=@n19.Random Offset)
   B=Panner(Coordinate=Multiply(A=@n15, B=@n21), Speed=AppendVector(A=@n20.B, B=@n20.B))
  )
  n23 = AppendVector(
   A=Multiply(A=ComponentMask(None=@n18.RGB) {g=true, r=true}, B=ScalarParameter() {defaultValue=1, group=WPO, parameterName=WPO Side Intensity}) {constB=10}
   B=Multiply(
    A=TextureSample(UVs=@n22) {texture=&T_Smooth_tile}.R
    B=ScalarParameter() {defaultValue=4, group=WPO, parameterName=WPO Wave Intensity}
   ) {constB=10}
  )
  n24 = ComponentMask(None=TextureCoordinate()) {r=true}
  n25 = LinearInterpolate(B=@n23, Alpha=@n24)
  n26 = ScalarParameter() {defaultValue=1, parameterName=Use_WPOMaskLerp}
  n27 = DynamicParameter() {defaultValue=<rgba 1 0 0 6>, paramNames [Random Offset, BaseTextureV_Add, WPO Wave Intensity, SweepTexture_Add]}
  NamedRerouteDeclaration(None=Multiply(A=LinearInterpolate(A=@n25, B=@n23, Alpha=@n26), B=@n27.WPO Wave Intensity) {constB=10}) {name=WPO, nodeColor=<rgba 0.0312 0 1 1>}
  n28 = DynamicParameter() {desc=Parameters set by Niagara, parameterIndex=1, paramNames [X,Y,Z,WPOIntensity]}
  n29 = Clamp(
   None=Subtract(
    A=MaterialFunctionCall(X=@n28.X, Y=@n28.X, Z=@n28.X) {materialFunction=&MakeFloat3}.Result
    B=ParticlePositionWS()
   )
   Min=ScalarParameter() {defaultValue=-25, group=WPO, parameterName=Clamp_min}
   Max=ScalarParameter() {defaultValue=25, group=WPO, parameterName=Clamp_max}
  ) {maxDefault=100, minDefault=-100}
  n30 = Add(A=TextureCoordinate(), B=AppendVector(A=ScalarParameter() {defaultValue=-0.5, parameterName=Interpolate WPO X Offset}, B=Constant()))
  n31 = Multiply(
   A=Power(Base=ComponentMask(None=@n30) {r=true}, Exp=ScalarParameter() {defaultValue=1.5, parameterName=InterpolateWPOBendPower})
   B=@n28.WPOIntensity
  )
  NamedRerouteDeclaration(None=LinearInterpolate(A=LinearInterpolate(B=@n29, Alpha=@n31), B=Constant(), Alpha=ScalarParameter() {defaultValue=1, parameterName=Use interpolate WPO})) {name=Interpolate WPO, nodeColor=<rgba 0 1 0.9515 1>}
  n32 = TextureSampleParameter2D(UVs=Panner(Coordinate=TextureCoordinate() {uTiling=0.6, vTiling=0.5}) {speedX=-0.25, speedY=0.15}) {
   group=DissolveMask
   parameterName=Dissolve B Texture
   samplerType=SAMPLERTYPE_Grayscale
   sortPriority=0
   texture=&LowResBlurredNoise
  }
  n33 = TextureSampleParameter2D(UVs=Panner(Coordinate=TextureCoordinate() {uTiling=0.6, vTiling=0.3}) {speedX=-0.25, speedY=0.15}) {
   group=DissolveMask
   parameterName=Dissolve A Texture
   samplerType=SAMPLERTYPE_LinearGrayscale
   sortPriority=0
   texture=/Game/Effects/Fort_Effects/Emotes/UltraLight/Texture/T_Plasma_C_Greyscale
  }
  n34 = ChannelMaskParameter(None=@n33.RGBA) {group=Base, parameterName=DissolveTexture_RGBselect, sortPriority=5}
  n35 = MaterialFunctionCall(
   UVs=TextureCoordinate()
   "Rotation Angle (0-1)"=ScalarParameter() {defaultValue=-0.06, group=DissolveMask, parameterName=DissolveMaskRotator}
  ) {materialFunction=&CustomRotator}
  n36 = DynamicParameter() {parameterIndex=2, defaultValue=<rgba 1 0 1 0>, paramNames [BaseTextureSpeedX_Offset, BaseTextureX_Add, Param3, DissolveMaskOffset]}
  n37 = AppendVector(
   A=ScalarParameter() {group=DissolveMask, parameterName=DissloveMaskAdd_X}
   B=Add(A=ScalarParameter() {defaultValue=-0.55, group=DissolveMask, parameterName=DissolveMaskAdd_Y}, B=@n36.DissolveMaskOffset)
  )
  n38 = Add(
   A=LinearInterpolate(A=@n32.R, B=Saturate(None=@n34.R), Alpha=ScalarParameter() {defaultValue=1, parameterName=DissolveTextureSelect})
   B=ComponentMask(None=Add(A=@n35.Rotated Values, B=@n37)) {g=true}
  )
  n39 = ScalarParameter() {defaultValue=3, parameterName=Dissolve_Int}
  NamedRerouteDeclaration(None=Saturate(None=Multiply(A=@n38, B=@n39) {constB=3})) {name=DissloveMask, nodeColor=<rgba 0.0312 0 1 1>}
  OneMinus(None=@n24)
  ScalarParameter() {group=Base, parameterName=BaseTextureU_Add}
  ScalarParameter() {group=Base, parameterName=BaseTextureV_Add}
  n40 = DynamicParameter() {parameterIndex=2, defaultValue=<rgba 1 0 1 0>, paramNames [BaseTextureSpeedX_Offset, BaseTextureX_Add, Param3, DissolveMaskOffset]}
  n41 = DynamicParameter() {defaultValue=<rgba 1 0 0 6>, paramNames [Random Offset, BaseTextureV_Add, WPO Wave Intensity, SweepTexture_Add]}
  n42 = Multiply(
   A=Add(A=TextureCoordinate(), B=AppendVector(A=@n40.BaseTextureX_Add, B=@n41.BaseTextureV_Add))
   B=AppendVector(
    A=ScalarParameter() {defaultValue=1, group=Base, parameterName=BaseTextureU}
    B=ScalarParameter() {defaultValue=1, group=Base, parameterName=BaseTextureV}
   )
  )
  n43 = TextureCoordinate() {uTiling=0.8}
  n44 = Constant2Vector() {r=0.2}
  n45 = ScalarParameter() {defaultValue=-0.1, group=BaseNoise, parameterName=NoiseTexture01Panner_X}
  n46 = ScalarParameter() {group=BaseNoise, parameterName=NoiseTexture01Panner_Y}
  n47 = TextureSampleParameter2D(UVs=Panner(Coordinate=Add(A=@n43, B=@n44), Speed=AppendVector(A=@n45, B=@n46)) {speedX=-0.1}) {
   group=BaseNoise
   parameterName=NoiseTexture01
   samplerType=SAMPLERTYPE_Grayscale
   texture=/BRCosmetics/Weapons/FORT_Melee/Pickaxe_Cube_Ninja_Male/FX/T_CellNoise
  }
  n48 = LinearInterpolate(Alpha=ComponentMask(None=@n47.RGB) {g=true, r=true}) {constA=0.2}
  n49 = Add(A=TextureCoordinate(), B=Constant2Vector())
  n50 = AppendVector(
   A=ScalarParameter() {defaultValue=-0.5, group=BaseNoise, parameterName=NoiseTexture02Panner_X}
   B=ScalarParameter() {group=BaseNoise, parameterName=NoiseTexture02Panner_Y}
  )
  n51 = ComponentMask(None=TextureSampleParameter2D(UVs=Panner(Coordinate=@n49, Speed=@n50) {speedX=-0.5}) {group=BaseNoise, parameterName=NoiseTexture02, texture=/Game/Effects/Fort_Effects/Emotes/UltraLight/Texture/T_SoftNoise_BC7}.RGB) {g=true, r=true}
  n52 = ComponentMask(None=TextureCoordinate()) {r=true}
  n53 = Multiply(
   A=Multiply(A=Add(A=@n48, B=@n51), B=LinearInterpolate(Alpha=@n52) {constA=0.1})
   B=ScalarParameter() {defaultValue=0.5, group=BaseNoise, parameterName=Noise_Intensity}
  ) {constB=0.3}
  n54 = Multiply(A=ScalarParameter() {defaultValue=-0.6, group=Base, parameterName=BaseTexture_SpeedX}, B=@n40.BaseTextureSpeedX_Offset)
  n55 = ScalarParameter() {group=Base, parameterName=BaseTexture_SpeedY}
  n56 = TextureSampleParameter2D(UVs=Panner(Coordinate=Add(A=@n42, B=@n53), Speed=AppendVector(A=@n54, B=@n55)) {speedX=-0.6}) {group=Base, parameterName=BaseTexture, texture=/BRCosmetics/Weapons/FORT_Melee/Pickaxe_LivelyDomino/FX/T_Pickaxe_LivelyDomino_trail_head}
  n57 = TextureSampleParameter2D(UVs=Panner(Coordinate=TextureCoordinate()) {speedX=-0.35, speedY=0.2}) {parameterName=BaseTextureLerp, samplerType=SAMPLERTYPE_Grayscale, texture=&LowResBlurredNoise}
  n58 = Multiply(
   A=Power(
    Base=ChannelMaskParameter(None=@n56.RGBA) {group=Base, parameterName=BaseTexture_RGBselect, sortPriority=5}.R
    Exp=ScalarParameter() {defaultValue=1, group=Base, parameterName=BaseTexture_power}
   )
   B=@n57.R
  )
  n59 = TextureSampleParameter2D() {parameterName=Mask, samplerSource=SSM_Clamp_WorldGroupSettings, texture=/Game/Effects/Fort_Effects/Textures/Mask/T_BaseLightMask_Half}
  n60 = LinearInterpolate(A=Multiply(A=@n58, B=@n59.R), B=@n58, Alpha=ScalarParameter() {defaultValue=1, group=DissolveMask, parameterName=Use_MaskTexture})
  n61 = NamedRerouteUsage()
  n62 = TextureCoordinate() {vTiling=0.5}
  n63 = ComponentMask(None=@n62) {r=true}
  n64 = OneMinus(None=@n63)
  n65 = ScalarParameter() {defaultValue=2, parameterName=Linear_Power}
  n66 = ScalarParameter() {defaultValue=1, parameterName=LinearLeft_Power}
  n67 = Multiply(A=Sine(None=ComponentMask(None=@n62) {g=true}), B=Multiply(A=Power(Base=@n64, Exp=@n65), B=Power(Base=@n63, Exp=@n66) {constExponent=1}))
  n68 = Multiply(
   A=LinearInterpolate(A=Multiply(A=@n60, B=@n61), B=@n60, Alpha=ScalarParameter() {defaultValue=1, group=DissolveMask, parameterName=Use_DissolveMask})
   B=Saturate(None=Multiply(A=@n67) {constB=2})
  )
  n69 = TextureCoordinate()
  n70 = MaterialFunctionCall(In=OneMinus(None=ComponentMask(None=@n69) {g=true}), Contrast=Constant() {r=0.3}) {materialFunction=&CheapContrast}
  n71 = Multiply(A=@n4.A, B=LinearInterpolate(A=Multiply(A=@n68, B=@n70.Result), B=@n68, Alpha=ScalarParameter() {defaultValue=1, parameterName=Use_Mask}))
  n72 = DepthFade(Opacity=@n71, FadeDistance=ScalarParameter() {defaultValue=5, parameterName=DepthFade})
  n73 = Multiply(
   A=@n71
   B=MaterialFunctionCall("FadeSensitivity (optional)"=ScalarParameter() {defaultValue=0.8, parameterName=MeshFade}) {materialFunction=&FadeBasedOnViewAngle}.Result
  )
  ShadingPathSwitch(Default=@n72, Deferred=@n72, Forward=@n73, Mobile=@n73)
 ]
}
